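# M_Chart is an unlit debug material in the Jump Flooding tools that sets up a plot of a stepped function.
# It uses the engine PlotFunctionOnGraph function over a 0 to 1 range, with line thickness as a scalar parameter.
# The function's x output is scaled, floored and divided back into quarter steps, then fed back as f(x).
# No connection from the plot to a material output is present.
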Material /Volumetrics/Tools/JumpFlooding/Materials/Debug/M_Chart
graph {
 properties {pixelDepthOffsetMode=PDOM_Legacy, shadingModel=MSM_Unlit, usedShadingModels=Unlit}
 nodes [
  n1 = Constant4Vector() {constant=<rgba 0 1 0 1>}
  n2 = Constant()
  n3 = Multiply(A=MaterialFunctionCall!(X range Y range=@n1.RGBA, Gradient Height=@n2, ShowGrid?, Thickness=ScalarParameter() {defaultValue=0.1}, Color, "f(x)"=@n4) {materialFunction=/Engine/Functions/Engine_MaterialFunctions01/Debug/PlotFunctionOnGraph}.x) {constB=4}
  n4 = Divide(A=Floor(None=Subtract(A=@n3) {constB=0})) {constB=4}
 ]
}
