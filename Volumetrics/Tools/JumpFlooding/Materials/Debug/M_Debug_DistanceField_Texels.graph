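# Unlit debug material that visualizes the jump-flooding distance field in the RT texture texel by texel.
# Point-sampled RG is read as a target UV, while alpha selects inner or outer tint and signs the offset.
# ShowDistance and ShowGradient blend from tinted raw data to distance or direction.
# A ShowGrid blend overlays an antialiased texel grid, target markers and procedural lines from texel centers toward their targets.
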Material /Volumetrics/Tools/JumpFlooding/Materials/Debug/M_Debug_DistanceField_Texels
graph {
 properties {pixelDepthOffsetMode=PDOM_Legacy, shadingModel=MSM_Unlit, usedShadingModels=Unlit}
 nodes [
  n1 = MaterialFunctionCall!(
   UVs=TextureCoordinate()
   Texture Size=TextureProperty(Tex=TextureObjectParameter() {parameterName=RT, texture=&Black})
  ) {materialFunction=/Engine/Functions/Engine_MaterialFunctions01/Texturing/PointSampledUVs}
  n2 = TextureSampleParameter2D!(UVs=@n1.Result, Level=Constant(), Apply View MipBias) {mipValueMode=TMVM_MipLevel, parameterName=RT, texture=&Black}
  n3 = ComponentMask(None=@n2.RGB) {g=true, r=true}
  n4 = Subtract(A=@n3, B=@n1.Result)
  n5 = Abs(None=@n4)
  n6 = Constant() {r=1}
  n7 = LinearInterpolate(Alpha=@n2.A) {constA=-1}
  n8 = Multiply(A=If(A=DotProduct(A=@n5, B=@n6), "A > B"=Normalize(VectorInput=@n4), "A == B"=@n4, "A < B"=@n4), B=@n7)
  Abs(None=@n8)
  AppendVector(A=@n1.Result, B=Constant())
  n9 = Multiply(
   A=@n2.RGB
   B=LinearInterpolate(
    A=ScalarParameter() {defaultValue=0.5, parameterName=OuterTint}
    B=ScalarParameter() {defaultValue=0.25, parameterName=InnerTint}
    Alpha=@n2.A
   ) {constA=0.5}
  )
  n10 = Distance(A=@n1.Result, B=@n3)
  n11 = ConstantBiasScale(None=Divide(A=Multiply(A=@n10, B=@n7), B=ScalarParameter() {defaultValue=0.1, parameterName=DistanceDivisor}))
  n12 = LinearInterpolate(
   A=LinearInterpolate(A=Max(A=@n9) {constB=0}, B=Saturate(None=@n11), Alpha=ScalarParameter() {parameterName=ShowDistance})
   B=AppendVector(A=@n8, B=Constant())
   Alpha=ScalarParameter() {parameterName=ShowGradient}
  )
  n13 = ScalarParameter() {defaultValue=0.2, parameterName=DebugSize}
  n14 = ComponentMask(None=TextureProperty(Tex=TextureObjectParameter() {parameterName=RT, texture=&Black})) {r=true}
  n15 = LinearInterpolate(
   A=@n12
   B=VectorParameter() {parameterName=DC, defaultValue=<rgba 0.5 0 0 0>}.RGB
   Alpha=SphereMask(A=@n3, B=TextureCoordinate(), Radius=Divide(A=@n13, B=@n14))
  )
  n16 = MaterialFunctionCall(
   UVs=TextureCoordinate()
   Texture Size=TextureProperty(Tex=TextureObjectParameter() {parameterName=RT, texture=&Black})
  ) {materialFunction=&PointSampledUVs}
  n17 = ComponentMask(None=TextureSampleParameter2D(UVs=@n16.Result, Level=Constant()) {mipValueMode=TMVM_MipLevel, parameterName=RT, texture=&Black}.RGB) {g=true, r=true}
  n18 = Subtract(A=@n17, B=@n16.Result)
  n19 = Constant2Vector() {r=1}
  n20 = If(A=DotProduct(A=Abs(None=@n18), B=Constant() {r=1}), "A > B"=Normalize(VectorInput=@n18), "A == B"=@n19, "A < B"=@n19)
  n21 = MaterialFunctionCall(Float2=@n20) {materialFunction=&BreakOutFloat2Components}
  n22 = Multiply(A=@n21.G) {constB=-1}
  n23 = TextureProperty(Tex=TextureObjectParameter() {parameterName=RT, texture=&Black})
  n24 = Multiply(A=Subtract(A=TextureCoordinate(), B=@n16.Result), B=@n23)
  n25 = ComponentMask(None=DotProduct(A=AppendVector(A=@n22, B=@n21.R), B=@n24)) {r=true}
  n26 = Constant2Vector()
  n27 = MaterialFunctionCall!(
   x=@n25
   h=MaterialFunctionCall!(In=AppendVector(A=@n25, B=@n26)) {materialFunction=/Engine/Functions/Engine_MaterialFunctions03/Procedurals/ComputeFilterWidth}.Result
   Split=ScalarParameter() {defaultValue=0.05, parameterName=Line Width}
  ) {materialFunction=/Engine/EditorMaterials/TilingAALineBoxFiltered}
  n28 = ComponentMask(None=@n20) {g=true, r=true}
  n29 = DotProduct(A=@n28, B=@n24)
  n30 = TextureCoordinate()
  n31 = Ceil(None=DotProduct(A=@n28, B=Subtract(A=@n30, B=@n17)))
  n32 = Multiply(A=Saturate(None=Ceil(None=@n29)), B=OneMinus(None=Saturate(None=@n31)))
  n33 = LinearInterpolate(B=OneMinus(None=Multiply(A=@n27.Result, B=@n32)), Alpha=ScalarParameter() {defaultValue=1, parameterName=LineOpacity}) {constA=1}
  n34 = MaterialFunctionCall!(ScalarUV=Multiply(A=TextureCoordinate(), B=@n23), Split=ScalarParameter() {defaultValue=0.02, parameterName=LineCutoff}) {materialFunction=/Engine/EditorMaterials/TilingAAGrid}
  LinearInterpolate(A=@n12, B=Add(A=Multiply(A=@n15, B=@n33), B=Multiply(A=@n34.Result)), Alpha=ScalarParameter() {defaultValue=0.2, parameterName=ShowGrid})
  Multiply()
  n35 = TextureCoordinate()
  n36 = TextureProperty(Tex=TextureObjectParameter() {parameterName=RT, texture=&Black})
  OneMinus(None=MaterialFunctionCall!(
    Coordinates=TextureCoordinate()
    P0=MaterialFunctionCall(UVs=@n35, Texture Size=@n36) {materialFunction=&PointSampledUVs}.Result
    P1=@n3
    Line Width=ScalarParameter() {defaultValue=0.0001, parameterName=Line Width Old}
    Edge Falloff
    Square Corners?=StaticBool()
   ) {materialFunction=/Engine/Functions/Engine_MaterialFunctions01/Coordinates/DrawLine-2D}.Result)
 ]
}
