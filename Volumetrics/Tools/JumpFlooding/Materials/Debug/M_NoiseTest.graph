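# M_NoiseTest is an unlit debug material in the JumpFlooding volumetrics tools for testing procedural noise in volume-texture space.
# Texture UVs pass through EncodeVolumeCoordinates with an XY frame count of 16, treating the flat UV layout as tiled slices of a volume.
# The resulting coordinate replaces world position as the Noise node's sample position.
# The graph shows no connection from the noise to a material output.
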Material /Volumetrics/Tools/JumpFlooding/Materials/Debug/M_NoiseTest
graph {
 properties {pixelDepthOffsetMode=PDOM_Legacy, shadingModel=MSM_Unlit, usedShadingModels=Unlit}
 nodes [Noise!(
   World Position=MaterialFunctionCall!(UVs=TextureCoordinate(), XY Frames=Constant() {r=16}) {materialFunction=/Volumetrics/Content/VolumeTextures/MaterialFunctions/EncodeVolumeCoordinates}.Result
   FilterWidth
  )]
}
